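# Idle Niagara effect for the Island Nomad backpack cosmetic: a glow tied to its lid, plus shed feathers.
# LidBoneRotation reads the user skeletal mesh, and its LidRotation scales light-beam sprite size, a particle light and a component-rendered HQ light.
# The particle light also follows a time-of-day curve.
# Feathers spawn while User.Jumped is true, falling with drag and sine-flattened flutter; a disabled FN_Debug emitter visualizes lid rotation.
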NiagaraSystem /BRCosmetics/Accessories/FORT_Backpacks/Backpack_F_MED_Island_Nomad/FX/NS_Island_Nomad_Idle
dependencies {
 usedDataInterfaces [&NiagaraDataInterfaceCurve, &NiagaraDataInterfaceVector2DCurve]
 usedDynamicInputs []
 usedModules [&SystemState]
}
emitters {
 FN_Debug {
  bEnabled=false
  emitterUpdateScript [
   &EmitterState()
   &SpawnBurst_Instantaneous(Spawn Count=1, Use Spawn Probability=false)
   &SpawnBurst_Instantaneous(Spawn Time=8.7309, Use Spawn Probability=false, UseLoopCountLimit=false) {module=SpawnBurst_Instantaneous001}
  ]
  particleSpawnScript [/Niagara/Modules/Spawn/Initialization/InitializeParticle(
    Lifetime=5
    Module.Write Color=true
    Module.Write Lifetime=true
    Module.Write Mass=true
    Module.Write Material Random=true
    Module.Write Position=false
    Module.Write Scale=false
    Module.Write SpriteRotation=false
    Module.Write SpriteSize=true
    Position=&SimulationPosition()
    Sprite Size=<100 100>
   )]
  particleUpdateScript [
   &ParticleState()
   &SkeletalMeshLocation(
    MaskSpawnGroup=false
    Mesh Sampling Type=::NewEnumerator5
    OverrideRandomSeed=false
    Sampled Position Offset=<0 0 60>
    Sampled Position Offset Coordinate Space:ENiagaraSkelMeshTransforms=::NewEnumerator0
    Skeletal Mesh=$User.Skeletal Mesh
   )
   &self:LidBoneRotation(New Skeletal Mesh:NiagaraDataInterfaceSkeletalMesh=$User.Skeletal Mesh) {module=LidBoneRotation}
   /SetVariables(Particles.Debug:NiagaraFloat=$Particles.LidRotation) {isSetParameters=true}
   &DynamicMaterialParameters(
    Index 0 Param 1=$Particles.Debug
    Module.Param0WriteEnabled=true
    Module.Param1WriteEnabled=true
    Module.Param2WriteEnabled=true
    Module.Param3WriteEnabled=true
   )
  ]
  rendererClasses [&NiagaraSpriteRendererProperties]
  renderers [{rendererClass=&NiagaraSpriteRendererProperties}]
 }
 Feathers {
  emitterUpdateScript [
   &EmitterState()
   &SpawnRate(
    Spawn Probability=0.75
    SpawnRate=&MakeCustomFloatFromBool(Bool=$User.Jumped, True Float=3, False Float=0)
    Use Spawn Probability=true
   )
  ]
  particleSpawnScript [
   /Niagara/Modules/Spawn/Initialization/V2/InitializeParticle(
    Color=<rgba 0.1392 0.3741 0.7031 1>
    Lifetime Min=1.5
    Lifetime Mode=::NewEnumerator1
    Sprite Rotation Angle=&RandomRangeFloat(Minimum=0, Maximum=1)
    Sprite Rotation Mode=::NewEnumerator1
    Sprite Size Mode=::NewEnumerator3
    Uniform Sprite Size=16
    UsePositionOffset=false
    Write Material Random=true
   )
   &SkeletalMeshLocation(Bone Sampling Mode=::NewEnumerator4, MaskSpawnGroup=false, Sampled Position Offset=<0 -15 10>, Skeletal Mesh=$User.Skeletal Mesh)
   &FN_SetSpawnGroup(Normalize Spawn Group=true, Number of Spawn Groups=6, Particles Per Group=1)
   &AddVelocity(Velocity=<-35 -20 -50>)
  ]
  particleUpdateScript [
   &ParticleState()
   &Drag(Use Linear Drag=true, Use Rotational Drag=false)
   &GravityForce(Gravity=<0 0 -200>)
   &SolveForcesAndVelocity(Clamp Velocity=false, Limit Acceleration=false)
   &SpriteRotationRate(Rotation Rate=&RandomRangeFloat(Minimum=100, Maximum=200))
   &ScaleSpriteSize(
    Non-Uniform Curve Scale=&MakeVector2D(
     X=1
     Y=&Sine(Normalized Angle=$Emitter.Age, Period=&RandomRangeFloat(Minimum=2, Maximum=1), Scale=1, Bias=0.1)
    )
    Non-Uniform Curve Sprite Scale=DataInterface() {propertyValues {
      XCurve {keys [
        {arriveTangent=8.2479, interpMode=RCIM_Cubic, leaveTangent=8.2479, tangentMode=RCTM_User, time=0, value=0.5}
        <at 0.1 1>
        {interpMode=RCIM_Cubic, tangentMode=RCTM_User, time=0.7, value=1}
        {tangentMode=RCTM_User, time=1, value=0}
       ]}
      YCurve {keys [
        {arriveTangent=8.2479, interpMode=RCIM_Cubic, leaveTangent=8.2479, tangentMode=RCTM_User, time=0, value=0.5}
        <at 0.1 1>
        {interpMode=RCIM_Cubic, tangentMode=RCTM_User, time=0.7, value=1}
        {tangentMode=RCTM_User, time=1, value=0}
       ]}
     }}
    Scale Sprite Size Mode=::NewEnumerator3
   )
  ]
  rendererClasses [&NiagaraSpriteRendererProperties]
  renderers [{rendererClass=&NiagaraSpriteRendererProperties}]
 }
 HQ_Light {
  emitterUpdateScript [
   &EmitterState()
   &SpawnBurst_Instantaneous(
    Spawn Group=1
    Spawn Probability=HlslExpression() {hlslExpression="Emitter.LoopCount == 0"}
    Spawn Time=0.03
    Use Spawn Probability=true
    UseLoopCountLimit=false
   ) {module=SpawnBurst_Instantaneous001}
  ]
  particleUpdateScript [
   &ParticleState(Kill Particles When Lifetime Has Elapsed=false, Let Infinitely Lived Particles Die When Emitter Deactivates=true)
   &SkeletalMeshLocation(
    MaskSpawnGroup=false
    Mesh Sampling Type=::NewEnumerator5
    OverrideRandomSeed=false
    Sampled Position Offset=<0 0 8>
    Skeletal Mesh=$User.Skeletal Mesh
   )
   &self:LidBoneRotation(New Skeletal Mesh:NiagaraDataInterfaceSkeletalMesh=$User.Skeletal Mesh) {module=LidBoneRotation}
   /SetVariables(Particles.Light_HQ_Alpha:NiagaraFloat=&Multiply_Float(A=$Particles.LidRotation, B=250)) {isSetParameters=true}
   &FN_Light_Attributes(Color=<rgba 0 0.5 1 1>, Create Color Attribute=true, Enable Particle Light Enable=true, Light Enabled=true) {module=FN_Light_Attributes001}
   &CameraOffset(Camera Offset Amount=50)
  ]
  rendererClasses [&NiagaraComponentRendererProperties]
  renderers [{rendererClass=&NiagaraComponentRendererProperties, rendererIndex=0}]
 }
 Light {
  emitterUpdateScript [
   &EmitterState()
   &SpawnBurst_Instantaneous(
    Spawn Group=1
    Spawn Probability=HlslExpression() {hlslExpression="Emitter.LoopCount == 0"}
    Spawn Time=0.03
    Use Spawn Probability=true
    UseLoopCountLimit=false
   ) {module=SpawnBurst_Instantaneous001}
  ]
  particleUpdateScript [
   &ParticleState(Kill Particles When Lifetime Has Elapsed=false, Let Infinitely Lived Particles Die When Emitter Deactivates=true)
   &SkeletalMeshLocation(
    MaskSpawnGroup=false
    Mesh Sampling Type=::NewEnumerator5
    OverrideRandomSeed=false
    Sampled Position Offset=<0 -7.5 2>
    Skeletal Mesh=$User.Skeletal Mesh
   )
   &self:LidBoneRotation(New Skeletal Mesh:NiagaraDataInterfaceSkeletalMesh=$User.Skeletal Mesh) {module=LidBoneRotation}
   &FN_Light_Attributes(
    Color=<rgba 0.5 2 5 1>
    Create Color Attribute=true
    Create Exponent Attribute=true
    Create Radius Attribute=true
    Exponent=0.5
    Radius=5
    Scale Color=&FloatFromCurve(
     FloatCurve=DataInterface() {propertyValues {Curve {keys [
         {interpMode=RCIM_Cubic, tangentMode=RCTM_User, time=0, value=0.5}
         {arriveTangent=1.3757, interpMode=RCIM_Cubic, leaveTangent=1.3757, tangentMode=RCTM_User, time=0.6, value=0.75}
         <at 1.0042 1>
        ]}}}
     CurveIndex=&self:ReferenceTimeOfDay()
     Scale Curve=&Multiply_Float(A=$Particles.LidRotation, B=5)
    )
   )
   &CameraOffset(Camera Offset Amount=50)
  ]
  rendererClasses [&NiagaraLightRendererProperties]
  renderers [{rendererClass=&NiagaraLightRendererProperties, rendererIndex=0}]
 }
 LightBeams {
  emitterUpdateScript [
   &EmitterState()
   &SpawnRate(SpawnRate=32, Use Spawn Probability=false) {module=SpawnRate002}
  ]
  particleSpawnScript [
   &InitializeParticle(
    Color Maximum=<rgba 0 0.25 0.5 0.35>
    Color Minimum=<rgba 0.5 0 0.2 0.35>
    Color Mode=::NewEnumerator2
    Lifetime Max=1
    Lifetime Min=0.5
    Lifetime Mode=::NewEnumerator1
    Position Mode=::NewEnumerator1
    Sprite Size=&Lerp_Vector2(
     A=<5 15>
     B=<10 25>
     Alpha=&Power(a=&RandomRangeFloat(Minimum=0, Maximum=1), b=3)
    )
    Sprite Size Mode=::NewEnumerator4
    UsePositionOffset=false
    Write Material Random=true
   )
   &SphereLocation(
    MaskSpawnGroup=false
    Non Uniform Scale=<5 1 1>
    Offset=<0 -10 40>
    Sphere Distribution:ENiagaraSphereDistributionMode=::NewEnumerator1
    Sphere Radius=5
    U Position:NiagaraFloat=&RandomRangeFloat(Minimum=0, Maximum=1)
    UseNonUniformScale=true
    UseOffset=true
    UseSphereOrientation=false
   )
   &DynamicMaterialParameters(
    Index 0 Param 1=&RandomRangeFloat(Minimum=-0.2, Maximum=0.2)
    Index 0 Param 2=1
    Index 0 Param 3=0.5
    Module.Param0WriteEnabled=true
    Module.Param1WriteEnabled=false
    Module.Param2WriteEnabled=false
    Module.Param3WriteEnabled=false
   )
   &AddVelocityFromPoint(Module.Use Falloff Distance=false, Velocity Strength=1)
  ]
  particleUpdateScript [
   &ParticleState()
   &SkeletalMeshLocation(MaskSpawnGroup=false, Mesh Sampling Type=::NewEnumerator5, OverrideRandomSeed=false, Skeletal Mesh=$User.Skeletal Mesh) {module=SkeletalMeshLocation001}
   &self:LidBoneRotation(New Skeletal Mesh:NiagaraDataInterfaceSkeletalMesh=$User.Skeletal Mesh) {module=LidBoneRotation}
   &Drag(Drag=5, Use Linear Drag=true, Use Rotational Drag=false)
   &SolveForcesAndVelocity(Clamp Velocity=false, Limit Acceleration=false)
   &ScaleColor(
    Scale Alpha=&FloatFromCurve(FloatCurve=DataInterface() {propertyValues {Curve {keys [
         {arriveTangent=6.668, interpMode=RCIM_Cubic, leaveTangent=6.668, tangentMode=RCTM_User, time=0, value=0}
         {interpMode=RCIM_Cubic, tangentMode=RCTM_User, time=0.5, value=1}
         <at 1 0>
        ]}}})
    ScaleA=true
    ScaleRGB=false
   )
   &ScaleSpriteSize(
    Uniform Curve Scale=&Multiply_Float(A=$Particles.LidRotation, B=1)
    Uniform Curve Sprite Scale=DataInterface() {propertyValues {Curve {keys [
        {interpMode=RCIM_Cubic, tangentMode=RCTM_User, time=0, value=0}
        {arriveTangent=-0.427, interpMode=RCIM_Cubic, leaveTangent=-0.427, tangentMode=RCTM_User, time=1, value=1.525}
       ]}}}
   )
   &SpriteRotationRate(Rotation Rate=50)
  ]
  rendererClasses [&NiagaraSpriteRendererProperties]
  renderers [{rendererClass=&NiagaraSpriteRendererProperties}]
 }
}
summary {
 systemName=NS_Island_Nomad_Idle
 emitters [
  {emitterName=LightBeams, rendererClasses [&NiagaraSpriteRendererProperties]}
  {emitterName=Light, rendererClasses [&NiagaraLightRendererProperties]}
  {emitterName=Feathers, rendererClasses [&NiagaraSpriteRendererProperties]}
  {bEnabled=false, emitterName=FN_Debug, rendererClasses [&NiagaraSpriteRendererProperties]}
  {emitterName=HQ_Light, rendererClasses [&NiagaraComponentRendererProperties]}
 ]
 userVariables [
  User.Burst_Ring {type=NiagaraInt32}
  User.IsJumping {type=NiagaraBool}
  User.Jump_Alpha {type=NiagaraFloat, defaultValue=0, description=""}
  User.Jumped {type=NiagaraBool}
  User.Skeletal Mesh {
   type=NiagaraDataInterfaceSkeletalMesh
   defaultValue {
    struct=NiagaraExt_VariableValue_DataInterface
    value {dataInterface=&self:NiagaraDataInterfaceSkeletalMesh_0}
   }
  }
 ]
}
