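# Physics asset for the Island Nomad backpack cosmetic, with simple box and capsule bodies for secondary motion of its dangling attachments.
# The lid, handles, jars, chisels, palette, crystals and slingshot constrain directly to the two-box pack body, mostly with tight angular limits.
# The slingshot instead rotates freely and carries a three-link band chain.
# Two three-link feather chains hang from the left handle.
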PhysicsAsset /BRCosmetics/Accessories/FORT_Backpacks/Backpack_F_MED_Island_Nomad/Meshes/F_MED_Island_Nomad_Pack_FE_PhysicsAsset1
bodies {
 dyn_band_01 [{
   shapeType=Box
   extentX=1.581
   extentY=0.2
   extentZ=5.2462
   shapeName=dyn_band_01_box
   center=<-0.2686 0.5087 0.005>
   rotation=<rot 0.8672 -69.9259 19.9645>
  }]
 dyn_band_02 [{
   shapeType=Capsule
   length=1.5744
   radius=0.5
   shapeName=dyn_band_02_capsule
   center=<-0.1118 -0.2735 -0.2707>
   rotation=<rot -0.8275 -90 -1.5522>
  }]
 dyn_band_base [{shapeType=Capsule, length=2.3301, radius=0.5, shapeName=dyn_band_base_capsule, center=<0.6926 -0.0409 0.7564>, rotation=<rot 0 0 94.9999>}]
 dyn_chisel_bk_MSTR [{
   shapeType=Capsule
   length=5.8462
   radius=0.8842
   shapeName=dyn_chisel_bk_MSTR_capsule
   center=<5.6555 0.6095 -0.4067>
   rotation=<rot 0 -81.7617 -85.5041>
  }]
 dyn_chisel_frt_MSTR [{
   shapeType=Capsule
   length=4.3416
   radius=0.6556
   shapeName=dyn_chisel_frt_MSTR_capsule
   center=<4.1126 0.1555 0.0888>
   rotation=<rot 0 91.5162 -87.3701>
  }]
 dyn_crystals_bk_MSTR [{
   shapeType=Capsule
   length=4.4057
   radius=3.4559
   shapeName=dyn_crystals_bk_MSTR_capsule
   center=<4.0084 0.3155 -0.2662>
   rotation=<rot 0 -97.926 -65.669>
  }]
 dyn_crystals_fr_MSTR [{
   shapeType=Capsule
   length=3.1355
   radius=1.1634
   shapeName=dyn_crystals_fr_MSTR_capsule
   center=<2.7263 -0.2057 0.0914>
   rotation=<rot 0 95.758 -86.3125>
  }]
 dyn_crystals_mid_MSTR [{
   shapeType=Capsule
   length=3.4698
   radius=2.6284
   shapeName=dyn_crystals_mid_MSTR_capsule
   center=<2.7321 -0.7499 -0.0731>
   rotation=<rot 0 81.0114 -72.8024>
  }]
 dyn_feather_bot_01 [{shapeType=Capsule, length=2.9598, radius=0.9406, shapeName=dyn_feather_bot_01_capsule, center=<1.9939 0 0>, rotation=<rot 90 0 0>}]
 dyn_feather_bot_02 [{shapeType=Capsule, length=1.7793, radius=1.1303, shapeName=dyn_feather_bot_02_capsule, center=<1.2992 0 0>, rotation=<rot 90 0 0>}]
 dyn_feather_bot_MSTR [{shapeType=Capsule, length=2.1256, radius=0.9726, shapeName=dyn_feather_bot_MSTR_capsule, center=<1.7024 0 0>, rotation=<rot 90 0 0>}]
 dyn_feather_top_01 [{shapeType=Capsule, length=3.3111, radius=0.64, shapeName=dyn_feather_top_01_capsule, center=<1.9939 0 0>, rotation=<rot 90 0 0>}]
 dyn_feather_top_02 [{shapeType=Capsule, length=2.1966, radius=0.713, shapeName=dyn_feather_top_02_capsule, center=<1.2992 0 0>, rotation=<rot 90 0 0>}]
 dyn_feather_top_MSTR [{shapeType=Capsule, length=2.5428, radius=0.5553, shapeName=dyn_feather_top_MSTR_capsule, center=<1.7024 0 0>, rotation=<rot 90 0 0>}]
 dyn_handle_l_MSTR [{
   shapeType=Capsule
   length=3.7269
   radius=0.7373
   shapeName=dyn_handle_l_MSTR_capsule
   center=<1.345 -0.2613 0.08>
   rotation=<rot 1.472 -90 1.5341>
  }]
 dyn_handle_r_MSTR [{
   shapeType=Capsule
   length=2.005
   radius=0.755
   shapeName=dyn_handle_r_MSTR_capsule
   center=<2.4388 -0.1428 0.0303>
   rotation=<rot 0 1.7534 5.0012>
  }]
 dyn_jar_in_r_MSTR [{
   shapeType=Capsule
   length=1.3861
   radius=0.6396
   shapeName=dyn_jar_in_r_MSTR_capsule
   center=<2.4585 0.0435 0.0443>
   rotation=<rot 44.4615 -89.9999 -5.4699>
  }]
 dyn_jar_mid_l_MSTR [{
   shapeType=Capsule
   length=1.1817
   radius=0.5
   shapeName=dyn_jar_mid_l_MSTR_capsule
   center=<2.268 0.0458 0.0467>
   rotation=<rot 44.4619 -89.9999 -5.4704>
  }]
 dyn_jar_mid_r_MSTR [{
   shapeType=Capsule
   length=1.4348
   radius=0.6139
   shapeName=dyn_jar_mid_r_MSTR_capsule
   center=<1.6641 -0.0994 -0.1002>
   rotation=<rot 44.452 -90.0001 13.7946>
  }]
 dyn_jar_out_l_MSTR [{
   shapeType=Capsule
   length=1.6963
   radius=0.9512
   shapeName=dyn_jar_out_l_MSTR_capsule
   center=<-2.3873 0.0553 -0.1008>
   rotation=<rot 0 10.655 -45.5964>
  }]
 dyn_lid_MSTR [{
   shapeType=Box
   extentX=1.5
   extentY=12.4975
   extentZ=25.1774
   shapeName=dyn_lid_MSTR_box
   center=<8.7477 0.797 0.0203>
   rotation=<rot -0.0477 -90 -0.2011>
  }]
 dyn_palette_MSTR [{
   shapeType=Capsule
   length=4.733
   radius=2.0548
   shapeName=dyn_palette_MSTR_capsule
   center=<3.9723 -0.6103 -0.0481>
   rotation=<rot 0 98.7077 -46.4475>
  }]
 dyn_slingshot_MSTR [{
   shapeType=Box
   extentX=4.9323
   extentY=4.5869
   extentZ=10.0545
   shapeName=dyn_slingshot_MSTR_box
   center=<1.397 -0.8707 1.2429>
   rotation=<rot -38.8647 19.3184 -2.091>
  }]
 pack [
  {shapeType=Box, extentX=21.1607, extentY=14.5, extentZ=11.5577, shapeName=pack_box, center=<-0.1377 -8.2919 -3.5842>}
  {shapeType=Box, extentX=2, extentY=11.5, extentZ=7.5, shapeName=pack_box, center=<-10.1941 -9.2732 -6.6111>, rotation=<rot -19.9999 0 0>}
 ]
}
constraints [
 {bone1Name=dyn_lid_MSTR, bone2Name=pack, swing1LimitDegrees=60, swing2LimitDegrees=0.2, twistLimitDegrees=0.5}
 {bone1Name=dyn_handle_r_MSTR, bone2Name=pack, swing1LimitDegrees=55, swing2LimitDegrees=0.5, twistLimitDegrees=0.5}
 {bone1Name=dyn_band_01, bone2Name=dyn_band_base, swing1LimitDegrees=25, swing2LimitDegrees=15, twistLimitDegrees=1}
 {bone1Name=dyn_band_02, bone2Name=dyn_band_01, swing1LimitDegrees=45, swing2LimitDegrees=10, twistLimitDegrees=1}
 {bone1Name=dyn_band_base, bone2Name=dyn_slingshot_MSTR, swing1LimitDegrees=5, swing2LimitDegrees=25, twistLimitDegrees=1}
 {
  bone1Name=dyn_slingshot_MSTR
  bone2Name=pack
  swing1LimitDegrees=45
  swing1Motion=Free
  swing2LimitDegrees=45
  swing2Motion=Free
  twistLimitDegrees=45
  twistMotion=Free
 }
 {bone1Name=dyn_jar_in_r_MSTR, bone2Name=pack, swing1LimitDegrees=5, swing2LimitDegrees=5, twistLimitDegrees=1}
 {bone1Name=dyn_jar_mid_r_MSTR, bone2Name=pack, swing1LimitDegrees=15, swing2LimitDegrees=25, twistLimitDegrees=1}
 {bone1Name=dyn_jar_mid_l_MSTR, bone2Name=pack, swing1LimitDegrees=15, swing2LimitDegrees=25, twistLimitDegrees=1}
 {bone1Name=dyn_jar_out_l_MSTR, bone2Name=pack, swing1LimitDegrees=15, swing2LimitDegrees=25, twistLimitDegrees=1}
 {bone1Name=dyn_chisel_bk_MSTR, bone2Name=pack, swing1LimitDegrees=10, swing2LimitDegrees=2, twistLimitDegrees=0.001}
 {bone1Name=dyn_chisel_frt_MSTR, bone2Name=pack, swing1LimitDegrees=10, swing2LimitDegrees=2, twistLimitDegrees=0.001}
 {bone1Name=dyn_palette_MSTR, bone2Name=pack, swing1LimitDegrees=10, swing2LimitDegrees=5, twistLimitDegrees=0.5}
 {bone1Name=dyn_handle_l_MSTR, bone2Name=pack, swing1LimitDegrees=45, swing2LimitDegrees=0.5, twistLimitDegrees=0.5}
 {bone1Name=dyn_feather_bot_MSTR, bone2Name=dyn_handle_l_MSTR, swing1LimitDegrees=20, swing2LimitDegrees=20, twistLimitDegrees=5}
 {bone1Name=dyn_feather_bot_01, bone2Name=dyn_feather_bot_MSTR, swing1LimitDegrees=20, swing2LimitDegrees=20, twistLimitDegrees=5}
 {bone1Name=dyn_feather_bot_02, bone2Name=dyn_feather_bot_01, swing1LimitDegrees=20, swing2LimitDegrees=20, twistLimitDegrees=5}
 {bone1Name=dyn_feather_top_MSTR, bone2Name=dyn_handle_l_MSTR, swing1LimitDegrees=20, swing2LimitDegrees=20, twistLimitDegrees=5}
 {bone1Name=dyn_feather_top_01, bone2Name=dyn_feather_top_MSTR, swing1LimitDegrees=20, swing2LimitDegrees=20, twistLimitDegrees=5}
 {bone1Name=dyn_feather_top_02, bone2Name=dyn_feather_top_01, swing1LimitDegrees=20, swing2LimitDegrees=20, twistLimitDegrees=5}
 {bone1Name=dyn_crystals_bk_MSTR, bone2Name=pack, swing1LimitDegrees=5, swing2LimitDegrees=15, twistLimitDegrees=0.2}
 {bone1Name=dyn_crystals_fr_MSTR, bone2Name=pack, swing1LimitDegrees=15, swing2LimitDegrees=2, twistLimitDegrees=0.2}
 {bone1Name=dyn_crystals_mid_MSTR, bone2Name=pack, swing1LimitDegrees=5, swing2LimitDegrees=15, twistLimitDegrees=0.2}
]
