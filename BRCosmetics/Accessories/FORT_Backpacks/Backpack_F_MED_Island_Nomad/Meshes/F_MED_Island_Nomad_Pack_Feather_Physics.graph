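# Physics asset for the Island Nomad backpack cosmetic, covering its pack, two handles and dangling feathers.
# Two boxes form the pack body, with left and right handle capsules constrained to it as hinges that swing on one axis, with the other swing and twist locked.
# Top and bottom feather chains, three capsules each, hang from the left handle with tight swing and twist limits at each joint.
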PhysicsAsset /BRCosmetics/Accessories/FORT_Backpacks/Backpack_F_MED_Island_Nomad/Meshes/F_MED_Island_Nomad_Pack_Feather_Physics
bodies {
 dyn_feather_bot_01 [{shapeType=Capsule, length=2.9598, radius=0.9406, shapeName=dyn_feather_bot_01_capsule, center=<1.9939 0 0>, rotation=<rot 90 0 0>}]
 dyn_feather_bot_02 [{shapeType=Capsule, length=1.7793, radius=1.1303, shapeName=dyn_feather_bot_02_capsule, center=<1.2992 0 0>, rotation=<rot 90 0 0>}]
 dyn_feather_bot_MSTR [{shapeType=Capsule, length=2.1256, radius=0.9726, shapeName=dyn_feather_bot_MSTR_capsule, center=<1.7024 0 0>, rotation=<rot 90 0 0>}]
 dyn_feather_top_01 [{shapeType=Capsule, length=3.3111, radius=0.64, shapeName=dyn_feather_top_01_capsule, center=<1.9939 0 0>, rotation=<rot 90 0 0>}]
 dyn_feather_top_02 [{shapeType=Capsule, length=2.1966, radius=0.713, shapeName=dyn_feather_top_02_capsule, center=<1.2992 0 0>, rotation=<rot 90 0 0>}]
 dyn_feather_top_MSTR [{shapeType=Capsule, length=2.5428, radius=0.5553, shapeName=dyn_feather_top_MSTR_capsule, center=<1.7024 0 0>, rotation=<rot 90 0 0>}]
 dyn_handle_l_MSTR [{
   shapeType=Capsule
   length=3.7269
   radius=0.7373
   shapeName=dyn_handle_l_MSTR_capsule
   center=<1.345 -0.2613 0.08>
   rotation=<rot 1.472 -90 1.5341>
  }]
 dyn_handle_r_MSTR [{
   shapeType=Capsule
   length=2.005
   radius=0.755
   shapeName=dyn_handle_r_MSTR_capsule
   center=<2.4388 -0.1428 0.0303>
   rotation=<rot 0 1.7534 5.0012>
  }]
 pack [
  {shapeType=Box, extentX=21.1607, extentY=14.5, extentZ=11.5577, shapeName=pack_box, center=<-0.1377 -8.2919 -3.5842>}
  {shapeType=Box, extentX=2, extentY=11.5, extentZ=7.5, shapeName=pack_box, center=<-10.1941 -9.2732 -6.6111>, rotation=<rot -19.9999 0 0>}
 ]
}
constraints [
 {
  bone1Name=dyn_handle_l_MSTR
  bone2Name=pack
  swing1LimitDegrees=45
  swing2LimitDegrees=0.5
  swing2Motion=Locked
  twistLimitDegrees=0.5
  twistMotion=Locked
 }
 {bone1Name=dyn_feather_bot_MSTR, bone2Name=dyn_handle_l_MSTR, swing1LimitDegrees=20, swing2LimitDegrees=20, twistLimitDegrees=5}
 {bone1Name=dyn_feather_bot_01, bone2Name=dyn_feather_bot_MSTR, swing1LimitDegrees=20, swing2LimitDegrees=20, twistLimitDegrees=5}
 {bone1Name=dyn_feather_bot_02, bone2Name=dyn_feather_bot_01, swing1LimitDegrees=20, swing2LimitDegrees=20, twistLimitDegrees=5}
 {bone1Name=dyn_feather_top_MSTR, bone2Name=dyn_handle_l_MSTR, swing1LimitDegrees=20, swing2LimitDegrees=20, twistLimitDegrees=5}
 {bone1Name=dyn_feather_top_01, bone2Name=dyn_feather_top_MSTR, swing1LimitDegrees=20, swing2LimitDegrees=20, twistLimitDegrees=5}
 {bone1Name=dyn_feather_top_02, bone2Name=dyn_feather_top_01, swing1LimitDegrees=20, swing2LimitDegrees=20, twistLimitDegrees=5}
 {
  bone1Name=dyn_handle_r_MSTR
  bone2Name=pack
  swing1LimitDegrees=55
  swing2LimitDegrees=0.5
  swing2Motion=Locked
  twistLimitDegrees=0.5
  twistMotion=Locked
 }
]
