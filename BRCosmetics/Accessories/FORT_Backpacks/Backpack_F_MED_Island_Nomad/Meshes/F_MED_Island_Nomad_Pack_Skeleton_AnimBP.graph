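# Post-process-style AnimBlueprint for the Island Nomad backpack cosmetic, adding secondary motion to its lid, feathers, crystals, tools and jars.
# Anim Dynamics and bone offsets pose the lid by skydiving state and Frontend_Alpha, cached as 'RBAN'.
# bRigidBodyEnabled selects RigidBody simulation, split by bIsInFrontEnd with rotator-driven crystals outside the front end, or an Anim Dynamics fallback chain.
# Several unconnected node chains also remain in the graph.
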AnimBlueprint /BRCosmetics/Accessories/FORT_Backpacks/Backpack_F_MED_Island_Nomad/Meshes/F_MED_Island_Nomad_Pack_Skeleton_AnimBP
target_skeleton=&/F_MED_Island_Nomad_Pack_Skeleton
graphs [
 AnimGraph {nodes [
   n1 = Anim Dynamics - Bone: dyn_slingshot_MSTR(Alpha=1, ExternalForce="0,0,0") {
    bGravityOverrideInSimSpace=true
    bUseGravityOverride=true
    simulationSpace=RootRelative
    alphaBoolBlend {blendOption=Linear}
    boundBone {boneName=dyn_slingshot_MSTR}
    componentAppliedLinearAccClamp=<6000 6000 6000>
    componentLinearAccScale=<0.15 0.2 0.3>
    componentLinearVelScale=<0.15 0.2 0.3>
    gravityOverride {z=-980}
    physicsBodyDefinitions [{
      sphereCollisionRadius=0
      boundBone {boneName=dyn_slingshot_MSTR}
      boxExtents=<5 10 10>
      constraintSetup {angularLimitsMax=<0 2 0>, angularLimitsMin=<0 -1 0>, angularTarget=<0 0 0>, linearAxesMax=<2 0 0>, linearAxesMin=<-0.2 0 0>}
      localJointOffset=<-3 0 0>
     }]
   }
   n2 = Anim Dynamics - Bone: dyn_band_base(ComponentPose=@n1, Alpha=0.15, ExternalForce="0,0,0") {
    angularDampingOverride=1
    angularSpringConstant=20000
    bGravityOverrideInSimSpace=true
    bOverrideAngularDamping=true
    bUseGravityOverride=true
    bUseSphericalLimits=true
    gravityScale=0.45
    linearSpringConstant=1000
    simulationSpace=RootRelative
    alphaBoolBlend {blendOption=Linear}
    boundBone {boneName=dyn_band_base}
    componentAppliedLinearAccClamp=<6000 6000 6000>
    componentLinearAccScale=<1 1.5 1.3>
    componentLinearVelScale=<1 1.5 1.3>
    gravityOverride {z=-980}
    physicsBodyDefinitions [{
      sphereCollisionRadius=0
      boundBone {boneName=dyn_band_base}
      boxExtents=<10 20 20>
      constraintSetup {angularLimitsMax=<0 0 0>, angularLimitsMin=<0 -25 0>, angularTarget=<0 0 0>, linearAxesMax=<0 0 0.8>}
      localJointOffset=<5 0 0>
     }]
   }
   n3 = Anim Dynamics - Bone: dyn_band_01(ComponentPose=@n2, Alpha=0.2, ExternalForce="0,0,0") {
    angularDampingOverride=1
    angularSpringConstant=20000
    bGravityOverrideInSimSpace=true
    bOverrideAngularDamping=true
    bUseGravityOverride=true
    bUseSphericalLimits=true
    gravityScale=0.45
    linearSpringConstant=1000
    simulationSpace=RootRelative
    alphaBoolBlend {blendOption=Linear}
    boundBone {boneName=dyn_band_01}
    componentAppliedLinearAccClamp=<6000 6000 6000>
    componentLinearAccScale=<1 1.5 1.3>
    componentLinearVelScale=<1 1.5 1.3>
    gravityOverride {z=-980}
    physicsBodyDefinitions [{
      sphereCollisionRadius=0
      boundBone {boneName=dyn_band_01}
      boxExtents=<10 20 20>
      constraintSetup {angularLimitsMax=<2 0 2>, angularLimitsMin=<-2 0 -90>, angularTarget=<0 0 0>, linearAxesMax=<0 0 0.35>, linearAxesMin=<0 -0.2 -0.5>}
      localJointOffset=<2 0 0>
     }]
    sphericalLimits [{limitRadius=4.5, drivingBone {boneName=None}, sphereLocalOffset=<-2 -12 0>}]
   }
   Anim Dynamics - Bone: dyn_band_02(ComponentPose=@n3, Alpha=0.2, ExternalForce="0,0,0") {
    angularDampingOverride=1
    angularSpringConstant=20000
    bGravityOverrideInSimSpace=true
    bOverrideAngularDamping=true
    bUseGravityOverride=true
    bUseSphericalLimits=true
    gravityScale=0.45
    linearSpringConstant=1000
    simulationSpace=RootRelative
    alphaBoolBlend {blendOption=Linear}
    boundBone {boneName=dyn_band_02}
    componentAppliedLinearAccClamp=<6000 6000 6000>
    componentLinearAccScale=<1.8 1.5 1.3>
    componentLinearVelScale=<1.8 1.5 1.3>
    gravityOverride {z=-980}
    physicsBodyDefinitions [{
      sphereCollisionRadius=0
      boundBone {boneName=dyn_band_02}
      boxExtents=<10 20 20>
      constraintSetup {angularLimitsMax=<2 0 1>, angularLimitsMin=<-2 0 -85>, angularTarget=<0 0 0>, linearAxesMax=<0 0 3>, linearAxesMin=<0 0 -2>}
      localJointOffset=<3 0 0>
     }]
    sphericalLimits [{limitRadius=3, drivingBone {boneName=None}, sphereLocalOffset=<-2 -12 0>}]
   }
   n4 = "Transform (Modify) Bone - Bone: dyn_lid_MSTR"!(ComponentPose, Alpha=1, Rotation="0.000,-15,0.000") -> (Pose) {
    rotationMode=BMM_Additive
    rotationSpace=BCS_BoneSpace
    scaleSpace=BCS_BoneSpace
    translationSpace=BCS_BoneSpace
    alphaBoolBlend {blendOption=Linear}
    boneToModify {boneName=dyn_lid_MSTR}
    rotation {yaw=-15}
   }
   Anim Dynamics - Bone: dyn_lid_MSTR(ComponentPose=@n4, Alpha=0.5, ExternalForce="0,0,0") {
    angularDampingOverride=1
    angularSpringConstant=1500
    bAngularSpring=true
    bGravityOverrideInSimSpace=true
    bOverrideAngularDamping=true
    bUseGravityOverride=true
    bUseSphericalLimits=true
    gravityScale=0.45
    simulationSpace=RootRelative
    alphaBoolBlend {blendOption=Linear}
    boundBone {boneName=dyn_lid_MSTR}
    componentAppliedLinearAccClamp=<2500 2500 2500>
    componentLinearAccScale=<0.2 0.3 1>
    componentLinearVelScale=<0.2 0.3 1>
    gravityOverride {z=-1100}
    physicsBodyDefinitions [{
      sphereCollisionRadius=0
      boundBone {boneName=dyn_lid_MSTR}
      boxExtents=<8 20 20>
      constraintSetup {angularLimitsMax=<0 0 0.1>, angularLimitsMin=<0 0 -86>, angularTarget=<0 0 0>}
      localJointOffset=<5 0 0>
     }]
   }
   Anim Dynamics - Bone: dyn_palette_MSTR(Alpha=1, ExternalForce="0,0,0") {
    angularSpringConstant=1000
    bAngularSpring=true
    bGravityOverrideInSimSpace=true
    bUseGravityOverride=true
    alphaBoolBlend {blendOption=Linear}
    boundBone {boneName=dyn_palette_MSTR}
    componentLinearAccScale=<1 1.8 3>
    componentLinearVelScale=<1 1.8 3>
    gravityOverride {z=-980}
    physicsBodyDefinitions [{
      sphereCollisionRadius=0
      boundBone {boneName=dyn_palette_MSTR}
      boxExtents=<5 10 10>
      constraintSetup {angularLimitsMax=<1 3 2>, angularLimitsMin=<-1.5 -2.5 -20>, angularTarget=<0 0 0>, linearAxesMin=<0 -2.5 0>}
      localJointOffset=<-1 0 0>
     }]
   }
   Anim Dynamics - Bone: dyn_lid_MSTR(Alpha=1, ExternalForce="0,0,0") {
    angularDampingOverride=1
    angularSpringConstant=1500
    bAngularSpring=true
    bGravityOverrideInSimSpace=true
    bOverrideAngularDamping=true
    bUseGravityOverride=true
    bUseSphericalLimits=true
    gravityScale=0.45
    simulationSpace=RootRelative
    alphaBoolBlend {blendOption=Linear}
    boundBone {boneName=dyn_lid_MSTR}
    componentAppliedLinearAccClamp=<2500 2500 2500>
    componentLinearAccScale=<0.2 0.3 1>
    componentLinearVelScale=<0.2 0.3 1>
    gravityOverride {z=-1100}
    physicsBodyDefinitions [{
      sphereCollisionRadius=0
      boundBone {boneName=dyn_lid_MSTR}
      boxExtents=<8 20 20>
      constraintSetup {angularLimitsMax=<0 0 0>, angularLimitsMin=<0 0 -110>, angularTarget=<0 0 0>}
      localJointOffset=<5 0 0>
     }]
   }
   Anim Dynamics - Bone: dyn_lid_MSTR(Alpha=1, ExternalForce="0,0,0") {
    angularDampingOverride=1
    angularSpringConstant=1500
    bAngularSpring=true
    bGravityOverrideInSimSpace=true
    bOverrideAngularDamping=true
    bUseGravityOverride=true
    bUseSphericalLimits=true
    gravityScale=0.45
    simulationSpace=BoneRelative
    alphaBoolBlend {blendOption=Linear}
    alphaScaleBiasClamp {bInterpResult=true, interpSpeedDecreasing=6, interpSpeedIncreasing=2}
    boundBone {boneName=dyn_lid_MSTR}
    componentAppliedLinearAccClamp=<2500 2500 2500>
    componentLinearAccScale=<0.2 0.3 1>
    componentLinearVelScale=<0.2 0.3 1>
    gravityOverride {z=-1100}
    physicsBodyDefinitions [{
      sphereCollisionRadius=0
      boundBone {boneName=dyn_lid_MSTR}
      boxExtents=<8 20 20>
      constraintSetup {angularLimitsMax=<0 0 43>, angularLimitsMin=<0 0 0>, angularTarget=<0 0 0>}
      localJointOffset=<5 0 0>
     }]
   }
   n5 = "Transform (Modify) Bone - Bone: dyn_lid_MSTR"(Alpha=1, Rotation="0.000,-58,-0.000") {
    rotationMode=BMM_Replace
    rotationSpace=BCS_BoneSpace
    scaleSpace=BCS_BoneSpace
    translationSpace=BCS_BoneSpace
    alphaBoolBlend {blendOption=Linear}
    boneToModify {boneName=dyn_lid_MSTR}
    rotation {yaw=-58}
   }
   Anim Dynamics - Bone: dyn_lid_MSTR(ComponentPose=@n5, Alpha=Get Lid_Variable() -> (Lid_Variable:double), ExternalForce="0,0,0") {
    angularDampingOverride=1
    angularSpringConstant=1500
    bAngularSpring=true
    bGravityOverrideInSimSpace=true
    bOverrideAngularDamping=true
    bUseGravityOverride=true
    bUseSphericalLimits=true
    gravityScale=0.45
    simulationSpace=BoneRelative
    alphaBoolBlend {blendOption=Linear}
    alphaScaleBiasClamp {bInterpResult=true, interpSpeedDecreasing=15, interpSpeedIncreasing=2}
    boundBone {boneName=dyn_lid_MSTR}
    componentAppliedLinearAccClamp=<2500 2500 2500>
    componentLinearAccScale=<0.2 0.3 1>
    componentLinearVelScale=<0.2 0.3 1>
    gravityOverride {z=-980}
    physicsBodyDefinitions [{
      sphereCollisionRadius=0
      boundBone {boneName=dyn_lid_MSTR}
      boxExtents=<20 8 20>
      constraintSetup {angularLimitsMax=<0 0.02 43>, angularLimitsMin=<0 -0.02 0>, angularTarget=<0 0 0>}
      localJointOffset=<5 0 0>
     }]
   }
   n6 = Anim Dynamics - Bone: dyn_feather_top_MSTR(Alpha=1, ExternalForce="0,0,0") {
    angularDampingOverride=0.9999
    angularSpringConstant=1000
    bAngularSpring=true
    bGravityOverrideInSimSpace=true
    bOverrideAngularDamping=true
    bUseGravityOverride=true
    gravityScale=0.45
    simulationSpace=RootRelative
    alphaBoolBlend {blendOption=Linear}
    boundBone {boneName=dyn_feather_top_MSTR}
    componentAppliedLinearAccClamp=<800 800 800>
    componentLinearAccScale=<1 1 1>
    componentLinearVelScale=<1 1 1>
    gravityOverride {z=-980}
    physicsBodyDefinitions [{
      sphereCollisionRadius=0
      boundBone {boneName=dyn_feather_top_MSTR}
      boxExtents=<14 21 21>
      constraintSetup {angularLimitsMax=<12 2 12>, angularLimitsMin=<-12 -2 -12>, angularTarget=<0 0 0>}
      localJointOffset=<0 0 -5>
     }]
    relativeSpaceBone {boneName=pack}
   }
   n7 = Anim Dynamics - Bone: dyn_feather_top_01(ComponentPose=@n6, Alpha=1, ExternalForce="0,0,0") {
    angularDampingOverride=0.9999
    angularSpringConstant=1000
    bAngularSpring=true
    bGravityOverrideInSimSpace=true
    bOverrideAngularDamping=true
    bUseGravityOverride=true
    bUseSphericalLimits=true
    gravityScale=0.45
    simulationSpace=BoneRelative
    alphaBoolBlend {blendOption=Linear}
    boundBone {boneName=dyn_feather_top_01}
    componentAppliedLinearAccClamp=<800 800 800>
    componentLinearAccScale=<1 1 1>
    componentLinearVelScale=<1 1 1>
    gravityOverride {z=-980}
    physicsBodyDefinitions [{
      sphereCollisionRadius=0
      boundBone {boneName=dyn_feather_top_01}
      boxExtents=<14 21 21>
      constraintSetup {angularLimitsMax=<12 2 12>, angularLimitsMin=<-12 -2 -12>, angularTarget=<0 0 0>}
      localJointOffset=<0 0 -5>
     }]
    relativeSpaceBone {boneName=pack}
   }
   n8 = Anim Dynamics - Bone: dyn_feather_top_02(ComponentPose=@n7, Alpha=1, ExternalForce="0,0,0") {
    angularDampingOverride=0.9999
    angularSpringConstant=1000
    bAngularSpring=true
    bGravityOverrideInSimSpace=true
    bOverrideAngularDamping=true
    bUseGravityOverride=true
    gravityScale=0.45
    simulationSpace=BoneRelative
    alphaBoolBlend {blendOption=Linear}
    boundBone {boneName=dyn_feather_top_02}
    componentAppliedLinearAccClamp=<800 800 800>
    componentLinearAccScale=<1 1 1>
    componentLinearVelScale=<1 1 1>
    gravityOverride {z=-980}
    physicsBodyDefinitions [{
      sphereCollisionRadius=0
      boundBone {boneName=dyn_feather_top_02}
      boxExtents=<14 21 21>
      constraintSetup {angularLimitsMax=<12 2 12>, angularLimitsMin=<-12 -2 -12>, angularTarget=<0 0 0>}
      localJointOffset=<0 0 -5>
     }]
    relativeSpaceBone {boneName=pack}
   }
   n9 = Anim Dynamics - Bone: dyn_feather_bot_MSTR(ComponentPose=@n8, Alpha=1, ExternalForce="0,0,0") {
    angularDampingOverride=0.9999
    angularSpringConstant=1000
    bAngularSpring=true
    bGravityOverrideInSimSpace=true
    bOverrideAngularDamping=true
    bUseGravityOverride=true
    gravityScale=0.45
    simulationSpace=RootRelative
    alphaBoolBlend {blendOption=Linear}
    boundBone {boneName=dyn_feather_bot_MSTR}
    componentAppliedLinearAccClamp=<800 800 800>
    componentLinearAccScale=<1 1 1>
    componentLinearVelScale=<1 1 1>
    gravityOverride {z=-980}
    physicsBodyDefinitions [{
      sphereCollisionRadius=0
      boundBone {boneName=dyn_feather_bot_MSTR}
      boxExtents=<14 21 21>
      constraintSetup {angularLimitsMax=<12 2 12>, angularLimitsMin=<-12 -2 -12>, angularTarget=<0 0 0>}
      localJointOffset=<0 0 -5>
     }]
    relativeSpaceBone {boneName=pack}
   }
   n10 = Anim Dynamics - Bone: dyn_feather_bot_01(ComponentPose=@n9, Alpha=1, ExternalForce="0,0,0") {
    angularDampingOverride=0.9999
    angularSpringConstant=1000
    bAngularSpring=true
    bGravityOverrideInSimSpace=true
    bOverrideAngularDamping=true
    bUseGravityOverride=true
    bUseSphericalLimits=true
    gravityScale=0.45
    simulationSpace=BoneRelative
    alphaBoolBlend {blendOption=Linear}
    boundBone {boneName=dyn_feather_bot_01}
    componentAppliedLinearAccClamp=<800 800 800>
    componentLinearAccScale=<1 1 1>
    componentLinearVelScale=<1 1 1>
    gravityOverride {z=-980}
    physicsBodyDefinitions [{
      sphereCollisionRadius=0
      boundBone {boneName=dyn_feather_bot_01}
      boxExtents=<14 21 21>
      constraintSetup {angularLimitsMax=<12 2 12>, angularLimitsMin=<-12 -2 -12>, angularTarget=<0 0 0>}
      localJointOffset=<0 0 -5>
     }]
    relativeSpaceBone {boneName=pack}
   }
   Anim Dynamics - Bone: dyn_feather_bot_02(ComponentPose=@n10, Alpha=1, ExternalForce="0,0,0") {
    angularDampingOverride=0.9999
    angularSpringConstant=1000
    bAngularSpring=true
    bGravityOverrideInSimSpace=true
    bOverrideAngularDamping=true
    bUseGravityOverride=true
    gravityScale=0.45
    simulationSpace=BoneRelative
    alphaBoolBlend {blendOption=Linear}
    boundBone {boneName=dyn_feather_bot_02}
    componentAppliedLinearAccClamp=<800 800 800>
    componentLinearAccScale=<1 1 1>
    componentLinearVelScale=<1 1 1>
    gravityOverride {z=-980}
    physicsBodyDefinitions [{
      sphereCollisionRadius=0
      boundBone {boneName=dyn_feather_bot_02}
      boxExtents=<14 21 21>
      constraintSetup {angularLimitsMax=<12 2 12>, angularLimitsMin=<-12 -2 -12>, angularTarget=<0 0 0>}
      localJointOffset=<0 0 -5>
     }]
    relativeSpaceBone {boneName=pack}
   }
   Anim Dynamics - Bone: dyn_lid_MSTR(Alpha=1, ExternalForce="0,0,0") {
    angularDampingOverride=1
    angularSpringConstant=1000
    bAngularSpring=true
    bGravityOverrideInSimSpace=true
    bLinearSpring=true
    bOverrideAngularDamping=true
    bUseGravityOverride=true
    bUseSphericalLimits=true
    gravityScale=0.45
    linearSpringConstant=1000
    simulationSpace=BoneRelative
    alphaBoolBlend {blendOption=Linear}
    boundBone {boneName=dyn_lid_MSTR}
    componentAppliedLinearAccClamp=<2500 2500 2500>
    componentLinearAccScale=<0.2 0.3 1>
    componentLinearVelScale=<0.2 0.3 1>
    gravityOverride {z=-1100}
    physicsBodyDefinitions [{
      sphereCollisionRadius=0
      boundBone {boneName=dyn_lid_MSTR}
      boxExtents=<20 8 20>
      constraintSetup {angularLimitsMax=<0.1 0.15 0>, angularLimitsMin=<-0.1 -0.15 -55>, angularTarget=<0 0 0>}
      localJointOffset=<10 0 0>
     }]
    relativeSpaceBone {boneName=pack}
   }
   n11 = Anim Dynamics - Bone: dyn_band_base(Alpha=0.15, ExternalForce="0,0,0") {
    angularDampingOverride=1
    angularSpringConstant=20000
    bAngularSpring=true
    bGravityOverrideInSimSpace=true
    bLinearSpring=true
    bOverrideAngularDamping=true
    bUseGravityOverride=true
    bUseSphericalLimits=true
    gravityScale=0.45
    linearSpringConstant=1000
    simulationSpace=RootRelative
    alphaBoolBlend {blendOption=Linear}
    boundBone {boneName=dyn_band_base}
    componentAppliedLinearAccClamp=<2500 2500 2500>
    componentLinearAccScale=<1 1.5 1.3>
    componentLinearVelScale=<1 1.5 1.3>
    gravityOverride {z=-980}
    physicsBodyDefinitions [{
      sphereCollisionRadius=0
      boundBone {boneName=dyn_band_base}
      boxExtents=<10 20 20>
      constraintSetup {angularLimitsMax=<0 0 0>, angularLimitsMin=<0 -25 0>, angularTarget=<0 0 0>, linearAxesMax=<0 0 0.8>}
      localJointOffset=<5 0 0>
     }]
   }
   n12 = Anim Dynamics - Bone: dyn_band_01(ComponentPose=@n11, Alpha=0.2, ExternalForce="0,0,0") {
    angularDampingOverride=1
    angularSpringConstant=20000
    bAngularSpring=true
    bGravityOverrideInSimSpace=true
    bLinearSpring=true
    bOverrideAngularDamping=true
    bUseGravityOverride=true
    bUseSphericalLimits=true
    gravityScale=0.45
    linearSpringConstant=1000
    simulationSpace=RootRelative
    alphaBoolBlend {blendOption=Linear}
    boundBone {boneName=dyn_band_01}
    componentAppliedLinearAccClamp=<2500 2500 2500>
    componentLinearAccScale=<1 1.5 1.3>
    componentLinearVelScale=<1 1.5 1.3>
    gravityOverride {z=-980}
    physicsBodyDefinitions [{
      sphereCollisionRadius=0
      boundBone {boneName=dyn_band_01}
      boxExtents=<10 20 20>
      constraintSetup {angularLimitsMax=<2 0 2>, angularLimitsMin=<-2 0 -90>, angularTarget=<0 0 0>, linearAxesMax=<0 0 0.35>, linearAxesMin=<0 -0.2 -0.5>}
      localJointOffset=<2 0 0>
     }]
    sphericalLimits [{limitRadius=4.5, drivingBone {boneName=None}, sphereLocalOffset=<-2 -12 0>}]
   }
   Anim Dynamics - Bone: dyn_band_02(ComponentPose=@n12, Alpha=0.2, ExternalForce="0,0,0") {
    angularDampingOverride=1
    angularSpringConstant=20000
    bAngularSpring=true
    bGravityOverrideInSimSpace=true
    bLinearSpring=true
    bOverrideAngularDamping=true
    bUseGravityOverride=true
    bUseSphericalLimits=true
    gravityScale=0.45
    linearSpringConstant=1000
    simulationSpace=RootRelative
    alphaBoolBlend {blendOption=Linear}
    boundBone {boneName=dyn_band_02}
    componentAppliedLinearAccClamp=<2500 2500 2500>
    componentLinearAccScale=<1.8 1.5 1.3>
    componentLinearVelScale=<1.8 1.5 1.3>
    gravityOverride {z=-980}
    physicsBodyDefinitions [{
      sphereCollisionRadius=0
      boundBone {boneName=dyn_band_02}
      boxExtents=<10 20 20>
      constraintSetup {angularLimitsMax=<2 0 1>, angularLimitsMin=<-2 0 -85>, angularTarget=<0 0 0>, linearAxesMax=<0 0 3>, linearAxesMin=<0 0 -2>}
      localJointOffset=<3 0 0>
     }]
    sphericalLimits [{limitRadius=3, drivingBone {boneName=None}, sphereLocalOffset=<-2 -12 0>}]
   }
   Anim Dynamics - Bone: dyn_brush_MSTR(Alpha=1, ExternalForce="0,0,0") {
    bGravityOverrideInSimSpace=true
    bLinearSpring=true
    bUseGravityOverride=true
    linearSpringConstant=2
    simulationSpace=RootRelative
    alphaBoolBlend {blendOption=Linear}
    boundBone {boneName=dyn_brush_MSTR}
    componentAppliedLinearAccClamp=<6000 6000 6000>
    componentLinearAccScale=<0.5 0.2 0.1>
    componentLinearVelScale=<0.5 0.2 0.1>
    gravityOverride {z=-980}
    physicsBodyDefinitions [{
      sphereCollisionRadius=0
      boundBone {boneName=dyn_brush_MSTR}
      boxExtents=<5 10 10>
      constraintSetup {angularLimitsMax=<4 1.5 2>, angularLimitsMin=<-2 -1 -25>, angularTarget=<0 0 0>}
      localJointOffset=<-5 0 0>
     }]
   }
   "Transform (Modify) Bone - Bone: dyn_lid_MSTR"(Alpha=1, Rotation="0.000,55,0.000") {
    rotationMode=BMM_Additive
    rotationSpace=BCS_BoneSpace
    scaleSpace=BCS_BoneSpace
    translationSpace=BCS_BoneSpace
    alphaBoolBlend {blendOption=Linear}
    boneToModify {boneName=dyn_lid_MSTR}
    rotation {yaw=55}
   }
   "Transform (Modify) Bone - Bone: dyn_lid_MSTR"(Alpha=1, Rotation="0.000,-15,0.000") {
    alpha=0
    rotationMode=BMM_Additive
    rotationSpace=BCS_BoneSpace
    scaleSpace=BCS_BoneSpace
    translationSpace=BCS_BoneSpace
    alphaBoolBlend {blendOption=Linear}
    alphaScaleBiasClamp {interpSpeedDecreasing=30, interpSpeedIncreasing=30}
    boneToModify {boneName=dyn_lid_MSTR}
    rotation {yaw=-15}
   }
   RigidBody() {
    overridePhysicsAsset=&/F_MED_Island_Nomad_Pack_FE_PhysicsAsset1
    alphaBoolBlend {blendOption=Linear}
   }
   n13 = RigidBody(ComponentPose=Local To Component(LocalPose=Use cached pose 'RBAN'())) {
    bClampLinearTranslationLimitToRefPose=true
    bOverrideWorldGravity=true
    overridePhysicsAsset=&/F_MED_Island_Nomad_Pack_Feather_Physics
    simulationSpace=BaseBoneSpace
    alphaBoolBlend {blendOption=Linear}
    baseBoneRef {boneName=pack}
    componentAppliedLinearAccClamp=<6000 6000 6000>
    componentLinearAccScale=<0.4 0.4 0.55>
    componentLinearVelScale=<0.4 0.4 0.55>
    overrideWorldGravity {z=-980}
   }
   n14 = RigidBody(ComponentPose=@n13) {
    bClampLinearTranslationLimitToRefPose=true
    bOverrideWorldGravity=true
    overridePhysicsAsset=&F_MED_Island_Nomad_Pack_FE_PhysicsAsset1
    simulationSpace=BaseBoneSpace
    alphaBoolBlend {blendOption=Linear}
    baseBoneRef {boneName=pack}
    componentLinearAccScale=<0.3 0.2 0.1>
    componentLinearVelScale=<0.3 0.2 0.1>
    overrideWorldGravity {z=-980}
   }
   n15 = Component To Local(ComponentPose=@n14)
   n16 = "Transform (Modify) Bone - Bone: dyn_crystals_fr_MSTR"!(
    ComponentPose=Local To Component(LocalPose=Use cached pose 'RBAN'())
    Alpha=1
    Rotation=Get Floating_Rotator1() -> (Floating_Rotator1:Rotator)
   ) -> (Pose) {
    rotationMode=BMM_Additive
    rotationSpace=BCS_BoneSpace
    scaleSpace=BCS_BoneSpace
    translationSpace=BCS_BoneSpace
    alphaBoolBlend {blendOption=Linear}
    boneToModify {boneName=dyn_crystals_fr_MSTR}
    rotation {yaw=55}
   }
   n17 = "Transform (Modify) Bone - Bone: dyn_crystals_mid_MSTR"!(ComponentPose=@n16, Alpha=1, Rotation=Get Floating_Rotator2() -> (Floating_Rotator2:Rotator)) -> (Pose) {
    rotationMode=BMM_Additive
    rotationSpace=BCS_BoneSpace
    scaleSpace=BCS_BoneSpace
    translationSpace=BCS_BoneSpace
    alphaBoolBlend {blendOption=Linear}
    boneToModify {boneName=dyn_crystals_mid_MSTR}
    rotation {yaw=55}
   }
   n18 = "Transform (Modify) Bone - Bone: dyn_crystals_bk_MSTR"!(ComponentPose=@n17, Alpha=1, Rotation=Get Floating_Rotator3() -> (Floating_Rotator3:Rotator)) -> (Pose) {
    rotationMode=BMM_Additive
    rotationSpace=BCS_BoneSpace
    scaleSpace=BCS_BoneSpace
    translationSpace=BCS_BoneSpace
    alphaBoolBlend {blendOption=Linear}
    boneToModify {boneName=dyn_crystals_bk_MSTR}
    rotation {yaw=55}
   }
   n19 = RigidBody(ComponentPose=@n18) {
    bClampLinearTranslationLimitToRefPose=true
    bOverrideWorldGravity=true
    overridePhysicsAsset=&/F_MED_Island_Nomad_Pack_PhysicsAsset
    simulationSpace=BaseBoneSpace
    alphaBoolBlend {blendOption=Linear}
    baseBoneRef {boneName=pack}
    componentLinearAccScale=<0.3 0.2 0.1>
    componentLinearVelScale=<0.3 0.2 0.1>
    overrideWorldGravity {z=-980}
   }
   n20 = RigidBody(ComponentPose=@n19) {
    bClampLinearTranslationLimitToRefPose=true
    bOverrideWorldGravity=true
    overridePhysicsAsset=&F_MED_Island_Nomad_Pack_Feather_Physics
    simulationSpace=BaseBoneSpace
    alphaBoolBlend {blendOption=Linear}
    baseBoneRef {boneName=pack}
    componentAppliedLinearAccClamp=<6000 6000 6000>
    componentLinearAccScale=<0.4 0.4 0.55>
    componentLinearVelScale=<0.4 0.4 0.55>
    overrideWorldGravity {z=-980}
   }
   n21 = Component To Local(ComponentPose=@n20)
   n22 = Get bIsInFrontEnd()
   n23 = Anim Dynamics - Bone: dyn_handle_l_MSTR(ComponentPose=Local To Component(LocalPose=Use cached pose 'RBAN'()), Alpha=1, ExternalForce="0,0,0") {
    angularSpringConstant=1000
    bGravityOverrideInSimSpace=true
    bUseGravityOverride=true
    gravityScale=0.45
    linearSpringConstant=100
    simulationSpace=BoneRelative
    alphaBoolBlend {blendOption=Linear}
    boundBone {boneName=dyn_handle_l_MSTR}
    componentAppliedLinearAccClamp=<5000 5000 5000>
    componentLinearAccScale=<0.2 0.3 0.7>
    componentLinearVelScale=<0.2 0.3 0.7>
    gravityOverride {z=-980}
    physicsBodyDefinitions [{
      sphereCollisionRadius=0
      boundBone {boneName=dyn_handle_l_MSTR}
      boxExtents=<8 20 20>
      constraintSetup {angularLimitsMax=<2 0.5 2>, angularLimitsMin=<-2 -0.5 -95>, angularTarget=<0 0 0>}
      localJointOffset=<4 0 0>
     }]
    relativeSpaceBone {boneName=pack}
   }
   n24 = Anim Dynamics - Bone: dyn_handle_r_MSTR(ComponentPose=@n23, Alpha=1, ExternalForce="0,0,0") {
    angularSpringConstant=1000
    bGravityOverrideInSimSpace=true
    bUseGravityOverride=true
    gravityScale=0.45
    linearSpringConstant=100
    simulationSpace=BoneRelative
    alphaBoolBlend {blendOption=Linear}
    boundBone {boneName=dyn_handle_r_MSTR}
    componentAppliedLinearAccClamp=<5000 5000 5000>
    componentLinearAccScale=<0.2 0.3 0.7>
    componentLinearVelScale=<0.2 0.3 0.7>
    gravityOverride {z=-980}
    physicsBodyDefinitions [{
      sphereCollisionRadius=0
      boundBone {boneName=dyn_handle_r_MSTR}
      boxExtents=<8 20 20>
      constraintSetup {angularLimitsMax=<2 0.5 2>, angularLimitsMin=<-2 -0.5 -95>, angularTarget=<0 0 0>}
      localJointOffset=<4 0 0>
     }]
    relativeSpaceBone {boneName=pack}
   }
   n25 = Anim Dynamics - Bone: dyn_feather_bot_MSTR(ComponentPose=@n24, Alpha=0.65, ExternalForce="0,0,0") {
    alpha=0.9
    angularSpringConstant=600
    bAngularSpring=true
    bGravityOverrideInSimSpace=true
    numSolverIterationsPostUpdate=4
    numSolverIterationsPreUpdate=16
    alphaBoolBlend {blendOption=Linear}
    boundBone {boneName=dyn_feather_bot_MSTR}
    physicsBodyDefinitions [{
      sphereCollisionRadius=0
      boundBone {boneName=dyn_feather_bot_MSTR}
      boxExtents=<15 5 5>
      constraintSetup {twistAxis=AxisZ, angularLimitsMax=<20 30 30>, angularLimitsMin=<-20 -2 -30>, angularTarget=<0 0 1>}
      localJointOffset=<0 0 0>
     }]
   }
   n26 = Anim Dynamics - Bone: dyn_feather_bot_01(ComponentPose=@n25, Alpha=0.65, ExternalForce="0,0,0") {
    alpha=0.9
    angularSpringConstant=600
    bAngularSpring=true
    bGravityOverrideInSimSpace=true
    numSolverIterationsPostUpdate=4
    numSolverIterationsPreUpdate=16
    alphaBoolBlend {blendOption=Linear}
    boundBone {boneName=dyn_feather_bot_01}
    physicsBodyDefinitions [{
      sphereCollisionRadius=0
      boundBone {boneName=dyn_feather_bot_01}
      boxExtents=<15 5 5>
      constraintSetup {twistAxis=AxisZ, angularLimitsMax=<20 30 30>, angularLimitsMin=<-20 -2 -30>, angularTarget=<0 0 1>}
      localJointOffset=<0 0 0>
     }]
   }
   n27 = Anim Dynamics - Bone: dyn_feather_bot_02(ComponentPose=@n26, Alpha=0.65, ExternalForce="0,0,0") {
    alpha=0.9
    angularSpringConstant=600
    bAngularSpring=true
    bGravityOverrideInSimSpace=true
    numSolverIterationsPostUpdate=4
    numSolverIterationsPreUpdate=16
    alphaBoolBlend {blendOption=Linear}
    boundBone {boneName=dyn_feather_bot_02}
    physicsBodyDefinitions [{
      sphereCollisionRadius=0
      boundBone {boneName=dyn_feather_bot_02}
      boxExtents=<15 5 5>
      constraintSetup {twistAxis=AxisZ, angularLimitsMax=<20 30 30>, angularLimitsMin=<-20 -2 -30>, angularTarget=<0 0 1>}
      localJointOffset=<0 0 0>
     }]
   }
   n28 = Anim Dynamics - Bone: dyn_feather_top_MSTR(ComponentPose=@n27, Alpha=1, ExternalForce="0,0,0") {
    alpha=0.9
    angularSpringConstant=600
    bAngularSpring=true
    bGravityOverrideInSimSpace=true
    numSolverIterationsPostUpdate=4
    numSolverIterationsPreUpdate=16
    alphaBoolBlend {blendOption=Linear}
    boundBone {boneName=dyn_feather_top_MSTR}
    physicsBodyDefinitions [{
      sphereCollisionRadius=0
      boundBone {boneName=dyn_feather_top_MSTR}
      boxExtents=<15 5 5>
      constraintSetup {twistAxis=AxisZ, angularLimitsMax=<20 30 30>, angularLimitsMin=<-20 -2 -30>, angularTarget=<0 0 1>}
      localJointOffset=<0 0 0>
     }]
   }
   n29 = Anim Dynamics - Bone: dyn_feather_top_01(ComponentPose=@n28, Alpha=1, ExternalForce="0,0,0") {
    alpha=0.9
    angularSpringConstant=600
    bAngularSpring=true
    bGravityOverrideInSimSpace=true
    numSolverIterationsPostUpdate=4
    numSolverIterationsPreUpdate=16
    alphaBoolBlend {blendOption=Linear}
    boundBone {boneName=dyn_feather_top_01}
    physicsBodyDefinitions [{
      sphereCollisionRadius=0
      boundBone {boneName=dyn_feather_top_01}
      boxExtents=<15 5 5>
      constraintSetup {twistAxis=AxisZ, angularLimitsMax=<20 30 30>, angularLimitsMin=<-20 -2 -30>, angularTarget=<0 0 1>}
      localJointOffset=<0 0 0>
     }]
   }
   n30 = Anim Dynamics - Bone: dyn_feather_top_02(ComponentPose=@n29, Alpha=1, ExternalForce="0,0,0") {
    alpha=0.9
    angularSpringConstant=600
    bAngularSpring=true
    bGravityOverrideInSimSpace=true
    numSolverIterationsPostUpdate=4
    numSolverIterationsPreUpdate=16
    alphaBoolBlend {blendOption=Linear}
    boundBone {boneName=dyn_feather_top_02}
    physicsBodyDefinitions [{
      sphereCollisionRadius=0
      boundBone {boneName=dyn_feather_top_02}
      boxExtents=<15 5 5>
      constraintSetup {twistAxis=AxisZ, angularLimitsMax=<20 30 30>, angularLimitsMin=<-20 -2 -30>, angularTarget=<0 0 1>}
      localJointOffset=<0 0 0>
     }]
   }
   n31 = Anim Dynamics - Bone: dyn_chisel_bk_MSTR(ComponentPose=@n30, Alpha=1, ExternalForce="0,0,0") {
    angularSpringConstant=1000
    bAngularSpring=true
    bGravityOverrideInSimSpace=true
    bUseGravityOverride=true
    alphaBoolBlend {blendOption=Linear}
    boundBone {boneName=dyn_chisel_bk_MSTR}
    componentLinearAccScale=<0.5 0.1 0.35>
    componentLinearVelScale=<0.5 0.1 0.35>
    gravityOverride {z=-980}
    physicsBodyDefinitions [{
      sphereCollisionRadius=0
      boundBone {boneName=dyn_chisel_bk_MSTR}
      boxExtents=<2 10 10>
      constraintSetup {angularLimitsMax=<4 3 15>, angularLimitsMin=<-2 -2 -10>, angularTarget=<0 0 0>, linearAxesMax=<0 1 0>, linearAxesMin=<-0.2 0 0>}
      localJointOffset=<-2 0 0>
     }]
   }
   n32 = Anim Dynamics - Bone: dyn_chisel_frt_MSTR(ComponentPose=@n31, Alpha=1, ExternalForce="0,0,0") {
    angularSpringConstant=1000
    bAngularSpring=true
    bGravityOverrideInSimSpace=true
    bUseGravityOverride=true
    simulationSpace=BoneRelative
    alphaBoolBlend {blendOption=Linear}
    boundBone {boneName=dyn_chisel_frt_MSTR}
    componentLinearAccScale=<0.5 0.1 0.35>
    componentLinearVelScale=<0.5 0.1 0.35>
    gravityOverride {z=-980}
    physicsBodyDefinitions [{
      sphereCollisionRadius=0
      boundBone {boneName=dyn_chisel_frt_MSTR}
      boxExtents=<2 10 10>
      constraintSetup {angularLimitsMax=<2 3 20>, angularLimitsMin=<-3 -2 -10>, angularTarget=<0 0 0>, linearAxesMax=<0 0 2>, linearAxesMin=<-0.5 0 0>}
      localJointOffset=<-3 0 0>
     }]
   }
   n33 = Anim Dynamics - Bone: dyn_jar_in_r_MSTR(ComponentPose=@n32, Alpha=1, ExternalForce="0,0,0") {
    angularSpringConstant=1000
    bAngularSpring=true
    bGravityOverrideInSimSpace=true
    bUseGravityOverride=true
    simulationSpace=RootRelative
    alphaBoolBlend {blendOption=Linear}
    boundBone {boneName=dyn_jar_in_r_MSTR}
    componentLinearAccScale=<0.1 0.2 0.5>
    componentLinearVelScale=<0.1 0.2 0.5>
    gravityOverride {z=-980}
    physicsBodyDefinitions [{
      sphereCollisionRadius=0
      boundBone {boneName=dyn_jar_in_r_MSTR}
      boxExtents=<5 20 20>
      constraintSetup {angularLimitsMax=<3 15 3>, angularLimitsMin=<-1 -2 -2>, angularTarget=<0 0 0>, linearAxesMax=<1 0 0>, linearAxesMin=<-0.5 0 0>}
      localJointOffset=<-4 0 0>
     }]
   }
   n34 = Anim Dynamics - Bone: dyn_jar_mid_r_MSTR(ComponentPose=@n33, Alpha=1, ExternalForce="0,0,0") {
    angularSpringConstant=1000
    bAngularSpring=true
    bGravityOverrideInSimSpace=true
    bUseGravityOverride=true
    simulationSpace=RootRelative
    alphaBoolBlend {blendOption=Linear}
    boundBone {boneName=dyn_jar_mid_r_MSTR}
    componentLinearAccScale=<0.1 0.2 0.5>
    componentLinearVelScale=<0.1 0.2 0.5>
    gravityOverride {z=-980}
    physicsBodyDefinitions [{
      sphereCollisionRadius=0
      boundBone {boneName=dyn_jar_mid_r_MSTR}
      boxExtents=<5 20 20>
      constraintSetup {angularLimitsMax=<3 15 12>, angularLimitsMin=<-1 -2 -3>, angularTarget=<0 0 0>, linearAxesMax=<2 0 0>, linearAxesMin=<-0.2 0 0>}
      localJointOffset=<-4 0 0>
     }]
   }
   n35 = Anim Dynamics - Bone: dyn_jar_mid_l_MSTR(ComponentPose=@n34, Alpha=1, ExternalForce="0,0,0") {
    angularSpringConstant=1000
    bAngularSpring=true
    bGravityOverrideInSimSpace=true
    bUseGravityOverride=true
    simulationSpace=RootRelative
    alphaBoolBlend {blendOption=Linear}
    boundBone {boneName=dyn_jar_mid_l_MSTR}
    componentLinearAccScale=<0.1 0.2 0.5>
    componentLinearVelScale=<0.1 0.2 0.5>
    gravityOverride {z=-980}
    physicsBodyDefinitions [{
      sphereCollisionRadius=0
      boundBone {boneName=dyn_jar_mid_l_MSTR}
      boxExtents=<5 20 20>
      constraintSetup {angularLimitsMax=<3 15 3>, angularLimitsMin=<-1 -2 -2>, angularTarget=<0 0 0>, linearAxesMax=<1 0 0>, linearAxesMin=<-0.5 0 0>}
      localJointOffset=<-4 0 0>
     }]
   }
   n36 = Anim Dynamics - Bone: dyn_jar_out_l_MSTR(ComponentPose=@n35, Alpha=1, ExternalForce="0,0,0") {
    angularSpringConstant=1000
    bAngularSpring=true
    bGravityOverrideInSimSpace=true
    bUseGravityOverride=true
    simulationSpace=RootRelative
    alphaBoolBlend {blendOption=Linear}
    boundBone {boneName=dyn_jar_out_l_MSTR}
    componentLinearAccScale=<0.1 0.2 0.5>
    componentLinearVelScale=<0.1 0.2 0.5>
    gravityOverride {z=-980}
    physicsBodyDefinitions [{
      sphereCollisionRadius=0
      boundBone {boneName=dyn_jar_out_l_MSTR}
      boxExtents=<5 20 20>
      constraintSetup {angularLimitsMax=<3 15 15>, angularLimitsMin=<-1 -2 -2>, angularTarget=<0 0 0>, linearAxesMax=<2 0 0>, linearAxesMin=<-0.2 0 0>}
      localJointOffset=<4 0 0>
     }]
   }
   Output Pose(Result=Blend Poses by bool(
     BlendPose_0=Blend Poses by bool(BlendPose_0=@n15, BlendPose_1=@n21, bActiveValue=@n22) {blendType=Linear}
     BlendPose_1=Component To Local(ComponentPose=@n36)
     bActiveValue=Get bRigidBodyEnabled()
    ) {blendType=Linear}) {layerGroup=None}
   n37 = "Transform (Modify) Bone - Bone: dyn_lid_MSTR"(ComponentPose=Local To Component(LocalPose=Use cached pose 'Source'()), Alpha=1, Rotation="0.000,45,0.000") {
    alpha=0
    rotationMode=BMM_Replace
    rotationSpace=BCS_BoneSpace
    scaleSpace=BCS_BoneSpace
    translationSpace=BCS_BoneSpace
    alphaBoolBlend {blendOption=Linear}
    boneToModify {boneName=dyn_lid_MSTR}
    rotation {yaw=45}
   }
   n38 = Anim Dynamics - Bone: dyn_lid_MSTR(ComponentPose=@n37, Alpha=1, ExternalForce="0,0,0") {
    alpha=0
    angularDampingOverride=1
    angularSpringConstant=1500
    bGravityOverrideInSimSpace=true
    bOverrideAngularDamping=true
    bUseSphericalLimits=true
    gravityScale=0.45
    simulationSpace=RootRelative
    alphaBoolBlend {blendOption=Linear}
    boundBone {boneName=dyn_lid_MSTR}
    componentAppliedLinearAccClamp=<3000 3000 3000>
    componentLinearAccScale=<0.2 0.3 0.5>
    componentLinearVelScale=<0.2 0.3 0.5>
    gravityOverride {z=-1100}
    physicsBodyDefinitions [{
      sphereCollisionRadius=0
      boundBone {boneName=dyn_lid_MSTR}
      boxExtents=<20 8 20>
      constraintSetup {angularLimitsMax=<0 0 10>, angularLimitsMin=<0 0 0>, angularTarget=<0 0 0>}
      localJointOffset=<5 0 0>
     }]
   }
   n39 = Get Frontend_Alpha() -> (Frontend_Alpha:double)
   n40 = "Transform (Modify) Bone - Bone: dyn_lid_MSTR"(ComponentPose=Local To Component(LocalPose=Use cached pose 'Source'()), Alpha=@n39, Rotation="0.000,45,-0.000") {
    rotationMode=BMM_Replace
    rotationSpace=BCS_BoneSpace
    scaleSpace=BCS_BoneSpace
    translationSpace=BCS_BoneSpace
    alphaBoolBlend {blendOption=Linear}
    alphaScaleBiasClamp {bInterpResult=true, interpSpeedDecreasing=-1.5, interpSpeedIncreasing=3.5}
    boneToModify {boneName=dyn_lid_MSTR}
    rotation {yaw=45}
   }
   n41 = "Transform (Modify) Bone - Bone: pack"!(ComponentPose=@n40, Alpha=@n39, Rotation="0.000,0.000,-3") -> (Pose) {
    rotationMode=BMM_Replace
    rotationSpace=BCS_BoneSpace
    scaleSpace=BCS_BoneSpace
    translationSpace=BCS_BoneSpace
    alphaBoolBlend {blendOption=Linear}
    alphaScaleBiasClamp {bInterpResult=true, interpSpeedDecreasing=-1.5, interpSpeedIncreasing=3.5}
    boneToModify {boneName=pack}
    rotation {roll=-3}
   }
   n42 = Anim Dynamics - Bone: dyn_lid_MSTR(ComponentPose=@n41, Alpha=1, ExternalForce="0,0,0") {
    angularDampingOverride=1
    angularSpringConstant=1000
    bGravityOverrideInSimSpace=true
    bOverrideAngularDamping=true
    bUseGravityOverride=true
    bUseSphericalLimits=true
    gravityScale=0.45
    linearSpringConstant=1000
    simulationSpace=BoneRelative
    alphaBoolBlend {blendOption=Linear}
    boundBone {boneName=dyn_lid_MSTR}
    componentAppliedLinearAccClamp=<5000 5000 5000>
    componentLinearAccScale=<0.2 0.3 0.7>
    componentLinearVelScale=<0.2 0.3 0.7>
    gravityOverride {z=-980}
    physicsBodyDefinitions [{
      sphereCollisionRadius=0
      boundBone {boneName=dyn_lid_MSTR}
      boxExtents=<20 8 20>
      constraintSetup {angularLimitsMax=<0 0 45>, angularLimitsMin=<0 0 0>, angularTarget=<0 0 0>}
      localJointOffset=<8 0 0>
     }]
    relativeSpaceBone {boneName=pack}
   }
   Save cached pose 'RBAN'(Pose=Blend Poses by bool(BlendPose_0=Component To Local(ComponentPose=@n38), BlendPose_1=Component To Local(ComponentPose=@n42), bActiveValue=Get bIsSkyDiving()) {blendType=Linear}) {cacheName=RBAN}
   n43 = Anim Dynamics - Bone: dyn_palette_MSTR(ComponentPose=Local To Component(LocalPose=Input Pose()), Alpha=1, ExternalForce="0,0,0") {
    angularSpringConstant=1000
    bGravityOverrideInSimSpace=true
    gravityScale=0.25
    simulationSpace=BoneRelative
    alphaBoolBlend {blendOption=Linear}
    boundBone {boneName=dyn_palette_MSTR}
    componentLinearAccScale=<0.1 0.1 0.1>
    componentLinearVelScale=<0.1 0.1 0.1>
    gravityOverride {z=-980}
    physicsBodyDefinitions [{
      sphereCollisionRadius=0
      boundBone {boneName=dyn_palette_MSTR}
      boxExtents=<5 10 10>
      constraintSetup {angularLimitsMax=<1 3 3>, angularLimitsMin=<-1.5 -2.5 -8>, angularTarget=<0 0 0>}
      localJointOffset=<-5 0 0>
     }]
    relativeSpaceBone {boneName=pack}
   }
   n44 = Anim Dynamics - Bone: dyn_hammer_MSTR(ComponentPose=@n43, Alpha=1, ExternalForce="0,0,0") {
    angularSpringConstant=1500
    bGravityOverrideInSimSpace=true
    bUseGravityOverride=true
    gravityScale=0.25
    simulationSpace=BoneRelative
    alphaBoolBlend {blendOption=Linear}
    boundBone {boneName=dyn_hammer_MSTR}
    componentLinearAccScale=<0.1 0.1 0.1>
    componentLinearVelScale=<0.1 0.1 0.1>
    gravityOverride {z=-980}
    physicsBodyDefinitions [{
      sphereCollisionRadius=0
      boundBone {boneName=dyn_hammer_MSTR}
      boxExtents=<8 20 20>
      constraintSetup {angularLimitsMax=<0.2 2 10>, angularLimitsMin=<-1 -5 -8>, angularTarget=<0 0 0>}
      localJointOffset=<-5 0 0>
     }]
   }
   n45 = Anim Dynamics - Bone: dyn_crystals_fr_MSTR(ComponentPose=@n44, Alpha=1, ExternalForce="0,0,0") {
    angularSpringConstant=1000
    bGravityOverrideInSimSpace=true
    bUseGravityOverride=true
    simulationSpace=BoneRelative
    alphaBoolBlend {blendOption=Linear}
    boundBone {boneName=dyn_crystals_fr_MSTR}
    componentLinearAccScale=<0.2 0.05 0.1>
    componentLinearVelScale=<0.2 0.05 0.1>
    gravityOverride {z=-980}
    physicsBodyDefinitions [{
      sphereCollisionRadius=0
      boundBone {boneName=dyn_crystals_fr_MSTR}
      boxExtents=<2 10 10>
      constraintSetup {angularLimitsMax=<4 3 15>, angularLimitsMin=<-2 -5 -20>, angularTarget=<0 0 0>}
      localJointOffset=<-5 0 0>
     }]
    relativeSpaceBone {boneName=root}
   }
   n46 = Anim Dynamics - Bone: dyn_crystals_mid_MSTR(ComponentPose=@n45, Alpha=1, ExternalForce="0,0,0") {
    angularSpringConstant=1000
    bGravityOverrideInSimSpace=true
    bUseGravityOverride=true
    simulationSpace=BoneRelative
    alphaBoolBlend {blendOption=Linear}
    boundBone {boneName=dyn_crystals_mid_MSTR}
    componentLinearAccScale=<0.2 0.05 0.1>
    componentLinearVelScale=<0.2 0.05 0.1>
    gravityOverride {z=-980}
    physicsBodyDefinitions [{
      sphereCollisionRadius=0
      boundBone {boneName=dyn_crystals_mid_MSTR}
      boxExtents=<2 10 10>
      constraintSetup {angularLimitsMax=<2 4 30>, angularLimitsMin=<-3 -8 -10>, angularTarget=<0 0 0>}
      localJointOffset=<-5 0 0>
     }]
    relativeSpaceBone {boneName=root}
   }
   n47 = Anim Dynamics - Bone: dyn_crystals_bk_MSTR(ComponentPose=@n46, Alpha=1, ExternalForce="0,0,0") {
    angularSpringConstant=1000
    bGravityOverrideInSimSpace=true
    bUseGravityOverride=true
    simulationSpace=BoneRelative
    alphaBoolBlend {blendOption=Linear}
    boundBone {boneName=dyn_crystals_bk_MSTR}
    componentLinearAccScale=<0.2 0.05 0.1>
    componentLinearVelScale=<0.2 0.05 0.1>
    gravityOverride {z=-980}
    physicsBodyDefinitions [{
      sphereCollisionRadius=0
      boundBone {boneName=dyn_crystals_bk_MSTR}
      boxExtents=<2 10 10>
      constraintSetup {angularLimitsMax=<2 15 0.2>, angularLimitsMin=<-3 -2 -8>, angularTarget=<0 0 0>}
      localJointOffset=<-5 0 0>
     }]
    relativeSpaceBone {boneName=root}
   }
   n48 = Anim Dynamics - Bone: dyn_brush_MSTR(ComponentPose=@n47, Alpha=1, ExternalForce="0,0,0") {
    bGravityOverrideInSimSpace=true
    bLinearSpring=true
    bUseGravityOverride=true
    linearSpringConstant=2
    simulationSpace=BoneRelative
    alphaBoolBlend {blendOption=Linear}
    boundBone {boneName=dyn_brush_MSTR}
    componentLinearAccScale=<0.15 0.3 0.5>
    componentLinearVelScale=<0.15 0.3 0.5>
    gravityOverride {z=-980}
    physicsBodyDefinitions [{
      sphereCollisionRadius=0
      boundBone {boneName=dyn_brush_MSTR}
      boxExtents=<5 10 10>
      constraintSetup {angularLimitsMax=<4 1.5 2>, angularLimitsMin=<-2 -1 -25>, angularTarget=<0 0 0>}
      localJointOffset=<-5 0 0>
     }]
    relativeSpaceBone {boneName=root}
   }
   n49 = Anim Dynamics - Bone: dyn_brush_01(ComponentPose=@n48, Alpha=0.4, ExternalForce="0,0,0") {
    angularDampingOverride=0.9999
    angularSpringConstant=1500
    bAngularSpring=true
    bGravityOverrideInSimSpace=true
    bOverrideAngularDamping=true
    gravityScale=0.75
    simulationSpace=BoneRelative
    alphaBoolBlend {blendOption=Linear}
    boundBone {boneName=dyn_brush_01}
    componentAppliedLinearAccClamp=<4500 4500 4500>
    componentLinearAccScale=<0.035 0.05 0.2>
    componentLinearVelScale=<0.035 0.05 0.2>
    physicsBodyDefinitions [{
      sphereCollisionRadius=0
      boundBone {boneName=dyn_brush_01}
      boxExtents=<10 20 20>
      constraintSetup {angularLimitsMax=<5 25 15>, angularLimitsMin=<-5 -30 -30>, angularTarget=<0 0 0>}
      localJointOffset=<2 0 0>
     }]
    relativeSpaceBone {boneName=root}
   }
   n50 = Anim Dynamics - Bone: dyn_brush_02(ComponentPose=@n49, Alpha=0.4, ExternalForce="0,0,0") {
    angularDampingOverride=0.9999
    angularSpringConstant=1500
    bAngularSpring=true
    bGravityOverrideInSimSpace=true
    bOverrideAngularDamping=true
    gravityScale=0.75
    simulationSpace=BoneRelative
    alphaBoolBlend {blendOption=Linear}
    boundBone {boneName=dyn_brush_02}
    componentAppliedLinearAccClamp=<4500 4500 4500>
    componentLinearAccScale=<0.035 0.05 0.2>
    componentLinearVelScale=<0.035 0.05 0.2>
    physicsBodyDefinitions [{
      sphereCollisionRadius=0
      boundBone {boneName=dyn_brush_02}
      boxExtents=<10 20 20>
      constraintSetup {angularLimitsMax=<5 25 15>, angularLimitsMin=<-5 -30 -30>, angularTarget=<0 0 0>}
      localJointOffset=<2 0 0>
     }]
    relativeSpaceBone {boneName=root}
   }
   Save cached pose 'Source'(Pose=Component To Local(ComponentPose=@n50)) {cacheName=Source}
  ]}
 EventGraph {nodes [
   Event BlueprintUpdateAnimation()
   > n1 = Sequence!(execute) -> (then_0, then_1, then_2:Exec, then_3:Exec)
   n2 = GetOwnerSkeletalMeshComponent()
   > n3 = Random Vector Function!(
    execute
    CharacterPartAnimInstance:Custom Character Part Anim Instance Object Reference
    SkelMeshComponent:Mesh Component Object Reference=@n2
    JointName:Name=None
    GameTimeInSeconds:double=GetGameTimeInSeconds()
    TimeOffset:double=0.2
    PerlinScale:double=0
    Frequency:Vector="0.9,1.1,0.75"
    Amplitude:Vector="1,0.2,1.2"
    Offset:Vector="0,0,1"
   ) -> (then, VectorOut:Vector)
   n4 = BreakVector(InVec=vector * float(A=@n3, B=16))
   > Set Floating_Rotator1(Floating_Rotator1:Rotator=MakeRotator(Roll=@n4, Pitch=@n4.Y, Yaw=@n4.Z)) -> (Output_Get:Rotator)
   > n5 = Random Vector Function(
    SkelMeshComponent=@n2
    JointName=None
    GameTimeInSeconds=GetGameTimeInSeconds()
    TimeOffset=0
    PerlinScale=0
    Frequency="0.8,1,0.45"
    Amplitude="1,0.2,1.4"
    Offset="0,0,0"
   )
   n6 = BreakVector(InVec=vector * float(A=@n5, B=12))
   > Set Floating_Rotator2(Floating_Rotator2:Rotator=MakeRotator(Roll=@n6, Pitch=@n6.Y, Yaw=@n6.Z)) -> (Output_Get:Rotator)
   > n7 = Random Vector Function(
    SkelMeshComponent=@n2
    JointName=None
    GameTimeInSeconds=GetGameTimeInSeconds()
    TimeOffset=0
    PerlinScale=0
    Frequency="0.9,1.35,0.75"
    Amplitude="2,0.2,1"
    Offset="0,0,0"
   )
   n8 = BreakVector(InVec=vector * float(A=@n7, B=5))
   > Set Floating_Rotator3(Floating_Rotator3:Rotator=MakeRotator(Roll=@n8, Pitch=@n8.Y, Yaw=@n8.Z)) -> (Output_Get:Rotator)
   n9 = Get bIsFalling()
   n10 = Get bIsJumping()
   Set Jump_Alpha(execute=@n1.then_1, Jump_Alpha:double=SelectFloat(A=1, B=0, bPickA=OR Boolean(A=@n10, B=@n9))) -> (Output_Get:double)
   Delay(execute=@n1.then_2, Duration=1)
   n11 = RandomBool!() -> (ReturnValue:bool)
   > Set Lid_Variable(Lid_Variable:double="Clamp (Float)"(Value="To Float (Boolean)"(InBool=@n11), Min=0.5, Max=1)) -> (Output_Get:double)
   n12 = Branch(execute=@n1.then_3, Condition=Get bIsInFrontEnd())
   > n13 = Branch(Condition=Get bIsPlayingEmote())
   > n14 = Set Frontend_Alpha(Frontend_Alpha:double=0) -> (Output_Get:double)
   Do Once(execute=@n13.else, Reset=[@n14.then, @n15.then])
   > Set Frontend_Alpha(Frontend_Alpha=0)
   > Delay(Duration=1.5)
   > Set Frontend_Alpha(Frontend_Alpha=1)
   > Delay(Duration=3.5)
   > Set Frontend_Alpha(Frontend_Alpha=0)
   n15 = Set Frontend_Alpha(execute=@n12.else, Frontend_Alpha=0)
   n16 = Branch(Condition=OR Boolean(A=Get bIsJumping(), B=Get bIsFalling()))
   > Set JumpingFalling(JumpingFalling:bool=true) -> (Output_Get:bool)
   Set JumpingFalling(execute=@n16.else, JumpingFalling=false)
   Branch(Condition=OR Boolean(A=Get bIsSkyDiving(), B=Get bIsInFrontEnd()))
   > Set frontEndorSkydiving(frontEndorSkydiving:bool=true) -> (Output_Get:bool)
   TryGetPawnOwner()
   MapRangeClamped(Value=0, InRangeA=0, InRangeB=0, OutRangeB=0)
  ]}
]
variables [
 Floating_Rotator1 {type=Rotator}
 Floating_Rotator2 {type=Rotator}
 Floating_Rotator3 {type=Rotator}
 Floating_Rotator4 {type=Rotator}
 Frontend_Alpha {type=number}
 frontEndorSkydiving {type=boolean}
 Jump_Alpha {type=number}
 JumpingFalling {type=boolean}
 Lid_Variable {type=number}
]
